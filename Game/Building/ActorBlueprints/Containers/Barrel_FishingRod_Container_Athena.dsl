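; Blueprint for the Athena fishing rod barrel container, overriding a few behaviors of its parent container class.
; Icon placement uses the actor's location plus an offset when the actor is valid, otherwise the parent's result.
; It is targetable by linear movers and disallows invisible props.
; The construction script adds a tiered-chest material tag, so level-placed instances receive the correct chrome material.
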(fn GetIconPlacement
  (params (SelfActor :type Actor%20Object%20Reference) (ViewingActor :type Actor%20Object%20Reference))
  (bind (_outlocation _outextents) (|CallParentFunction SelfActor ViewingActor))
  (Utilities|IsValid SelfActor
    (:Is%20Valid
      (return (+ (Transformation|GetActorLocation SelfActor) (LocalVariables|GetIconPlacementOffset)) _outextents))
    (:Is%20Not%20Valid
      (return _outlocation _outextents))))

(fn IsTargetableByLinearMover
  (return true))

(fn ShouldInvisiblePropsBeAllowed
  (return false))

(fn ConstructionScript
  ;[0.32 0.9402 1] Make sure actors that are already placed in levels get this tag so that they get the right chrome material added to them.
  (|CallParentFunction)
  (Utilities|Array|AddUnique (Variables|Actor|GetTags) "Lackluster.Material.TieredChestAthena"))
