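; Creative-mode tiered loot chest with cosmetic aura, light, ambient audio and a timed visual sound indicator until opened.
; Searching, looting or destruction clears those cosmetics; looting also spawns a rarity-tinted burst and open sound.
; Mobile highlighting uses timelines to drive an interact icon and material wiggle.
; Island resource manager requests gate light, VFX and audio; Outpost controllers need ownership to interact or destroy; prop-disguise destruction clears items.
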(event Custom|Chime%20Visual
  ;[1 1 1] Begin Play
  (if (Variables|Container|GetbAlreadySearched)
    (Utilities|Time|ClearandInvalidateTimerbyHandle (Variables|Default|GetChimeTimer))
    (else
      (Fortnite|ShowSoundIndicatorLocal :Location (Transformation|GetActorLocation) :MaxAudibleDistance (CallFunction|GetMaxAudibleDistance) :IndicatorType "Chest" :Instigator self :Tint "(R=0.000000,G=0.000000,B=0.000000,A=0.000000)" :AttenuationMaxRangeMultiplier 1.0 :AttenuationIntensityMultiplier 1.0))))

(event EventBeginPlay
  ;[1 1 1] Begin Play
  (Networking|CanExecuteCosmeticEvents
    (:True
      (CallFunction|CreateMobileMIDs)
      (if (Variables|Container|GetbAlreadySearched)
        (CallFunction|SetLightVisibility false)
        (else
          (CallFunction|SetLightVisibility true)
          (bind _returnvalue (Utilities|Time|SetTimerbyEvent (delegate Custom|Chime%20Visual) (Variables|Default|GetChestChimeVisualUpdate) true))
          (Variables|Default|SetChimeTimer _returnvalue))))))

(event Custom|Mobile_OnInteract
  ;[1 1 1] Mobile Wiggle on Interact
  (Networking|CanExecuteCosmeticEvents
    (:True
      (bind (_direction _lerp) (Timeline|MobileOnInteractTL Play)
        (:Update
          (for _array_element (Variables|Default|GetMIDs)
            (Rendering|Material|SetScalarParameterValue _array_element "WiggleOnInteractAmount" (Math|Float|Lerp 0.0 (Variables|Default|GetMobileWiggleAmount) _lerp) :!class /Script/Engine.MaterialInstanceDynamic)))))))

(event Custom|Mobile_Outline_On
  ;[1 1 1] Mobile Outline
  (for _array_element (Variables|Default|GetMIDs)
    (Rendering|Material|SetScalarParameterValue _array_element "Outline_Control" 1.0 :!class /Script/Engine.MaterialInstanceDynamic)
    (Rendering|Material|SetScalarParameterValue _array_element "Pulse_Control" 1.0 :!class /Script/Engine.MaterialInstanceDynamic)
    (Rendering|Material|SetVectorParameterValue _array_element "Color" "(R=1.000000,G=1.000000,B=1.000000,A=1.000000)" :!class /Script/Engine.MaterialInstanceDynamic)))

(event Container|EventEnableBacchusHighlight
  ;[1 1 1] Mobile highlighting
  (if (not (Variables|Container|GetbAlreadySearched))
    (Networking|CanExecuteCosmeticEvents
      (:True
        (CallFunction|SetVisibleMobileInteractIcon true)
        (bind (_direction _lerpobject _lerpinteactoicon) (Timeline|MobileSelectedTL Play)
          (:Update
            (bind _mobileinteracticon (Variables|Default|GetMobileInteractIcon)
              (:then
                (Transformation|SetRelativeScale3D _mobileinteracticon (Math|Vector|Lerp(Vector) "0, 0, 0" (Variables|Default|GetMobileInteractIconScale) _lerpinteactoicon)))
              (:else)))
          (:Finished
            (switch Utilities|FlowControl|Switch|SwitchonETimelineDirection _direction
              (:Backward
                (CallFunction|SetVisibleMobileInteractIcon false)))))))))

(event Custom|Mobile_OnDisconnect
  ;[1 1 1] Mobile Wiggle on Interact
  (Networking|CanExecuteCosmeticEvents
    (:True
      (Timeline|MobileOnInteractTL Reverse))))

(event Custom|Mobile_Outline_Off
  ;[1 1 1] Mobile Outline
  (for _array_element (Variables|Default|GetMIDs)
    (Rendering|Material|SetScalarParameterValue _array_element "Outline_Control" 0.0 :!class /Script/Engine.MaterialInstanceDynamic)
    (Rendering|Material|SetScalarParameterValue _array_element "Pulse_Control" 0.0 :!class /Script/Engine.MaterialInstanceDynamic)))

(event Container|EventDisableBacchusHighlight
  ;[1 1 1] Mobile highlighting
  (if (TouchPlatform|IsMobileGame)
    (Networking|CanExecuteCosmeticEvents
      (:True
        (Timeline|MobileSelectedTL Reverse)))))

(event EventOnBeginSearch
  ;[1 1 1] Mobile highlighting
  (if (Variables|Container|GetbAlreadySearched)
    _
    (elif (TouchPlatform|IsMobileGame)
      (Networking|CanExecuteCosmeticEvents
        (:True
          (Timeline|MobileSelectedTL Reverse))))))

(event EventOnSetSearched
  (Components|DestroyComponent (Variables|Default|GetNS_Athena_Loot_Chest_Aura))
  (Components|DestroyComponent (Variables|Default|GetS_Chest_SmokeSheet))
  (CallFunction|SetLightVisibility false)
  (Audio|Components|Audio|FadeOut (Variables|Default|GetChest_Ambient_Sound) 0.5 0.0 "Linear" :!class /Script/Engine.AudioComponent)
  (Utilities|Time|ClearandInvalidateTimerbyHandle (Variables|Default|GetChimeTimer))
  (Variables|Building|SetbDestroyOnPlayerBuildingPlacement true) ; Allows players to build over the opened chest mesh (but not when there is loot in there)
  (CallFunction|CleanupWiggleMIDs)
  (Components|Timeline|Stop (Variables|Creative_Tiered_Chest|GetMobileSelectedTL))
  (CallFunction|SetVisibleMobileInteractIcon false)
  (Rendering|SetRenderCustomDepth (Variables|Debug|GetStaticMeshComponent) false))

(event EventDestroyed
  (Components|DestroyComponent (Variables|Default|GetNS_Athena_Loot_Chest_Aura))
  (Components|DestroyComponent (Variables|Default|GetS_Chest_SmokeSheet))
  (CallFunction|SetLightVisibility false)
  (Audio|Components|Audio|FadeOut (Variables|Default|GetChest_Ambient_Sound) 0.5 0.0 "Linear" :!class /Script/Engine.AudioComponent)
  (Utilities|Time|ClearandInvalidateTimerbyHandle (Variables|Default|GetChimeTimer))
  (Variables|Building|SetbDestroyOnPlayerBuildingPlacement true) ; Allows players to build over the opened chest mesh (but not when there is loot in there)
  (CallFunction|CleanupWiggleMIDs)
  (Components|Timeline|Stop (Variables|Creative_Tiered_Chest|GetMobileSelectedTL))
  (CallFunction|SetVisibleMobileInteractIcon false)
  (Rendering|SetRenderCustomDepth (Variables|Debug|GetStaticMeshComponent) false))

(event EventOnLoot
  (bind _returnvalue (Transformation|GetActorLocation))
  (Utilities|FlowControl|Sequence
    (:then_0
      (Components|DestroyComponent (Variables|Default|GetNS_Athena_Loot_Chest_Aura))
      (Components|DestroyComponent (Variables|Default|GetS_Chest_SmokeSheet))
      (CallFunction|SetLightVisibility false)
      (Audio|Components|Audio|FadeOut (Variables|Default|GetChest_Ambient_Sound) 0.5 0.0 "Linear" :!class /Script/Engine.AudioComponent)
      (Utilities|Time|ClearandInvalidateTimerbyHandle (Variables|Default|GetChimeTimer))
      (Variables|Building|SetbDestroyOnPlayerBuildingPlacement true) ; Allows players to build over the opened chest mesh (but not when there is loot in there)
      (CallFunction|CleanupWiggleMIDs)
      (Components|Timeline|Stop (Variables|Creative_Tiered_Chest|GetMobileSelectedTL))
      (CallFunction|SetVisibleMobileInteractIcon false)
      (Rendering|SetRenderCustomDepth (Variables|Debug|GetStaticMeshComponent) false))
    (:then_1
      (bind _returnvalue_2 (Effects|SpawnEmitteratLocation "/Game/Effects/Fort_Effects/World/Chests/P_Loot_Chest_T1.P_Loot_Chest_T1" _returnvalue (Math|Rotator|MakeRotator (.roll (Transformation|GetActorRotation)) (+ (.pitch (Transformation|GetActorRotation)) 90) (.yaw (Transformation|GetActorRotation)))))
      (Variables|Default|SetLoot_Effect _returnvalue_2)
      (bind _returnvalue_3 (Fortnite|GetRarityData (Variables|Container|GetHighestRarity)))
      (Effects|Components|ParticleSystem|SetVectorParameter (Variables|Default|GetLoot_Effect) "color" (Math|Conversions|ToVector(LinearColor) (.Color2 (Utilities|Struct|BreakFortRarityItemData _returnvalue_3))))
      (Utilities|FlowControl|DoOnce false
        (:Completed
          (Audio|PlaySoundatLocation "/Game/Sounds/Foley_Loot/Containers/Treasure_Chest/Tiered_Chest_Open_T01_Cue.Tiered_Chest_Open_T01_Cue" _returnvalue))))))

(event CreativePlayerProp|EventPreDestroy
  ;[0 0 0 1 10] When the player is a chest prop, don't spawn the items when the prop is destroyed
  (ItemData|ClearItems (Variables|BuildingCustomizableSpawnContainer|GetItemListComponent)))

(event RequestResourcesDelegate(CreativeIslandResource)
  (params (ResourceManager :type Creative%20Island%20Resource%20Manager%20Component%20Object%20Reference))
  (CallFunction|OnRequestResources ResourceManager))

(event FreeResourcesDelegate(CreativeIslandResource)
  (params (ResourceManager :type Creative%20Island%20Resource%20Manager%20Component%20Object%20Reference))
  (CallFunction|OnFreeResources ResourceManager))

(fn BlueprintCanInteract
  (params (InteractingPawn :type Fort%20Pawn%20Object%20Reference) (InteractionBeingAttempted :type EInteractionBeingAttempted%20Enum :default "FirstInteraction") (InteractionType :type TInteractionType%20Enum :default "IT_NoInteraction"))
  (bind _asfort_player_controller_outpost (Utilities|Casting|CastToFortPlayerControllerOutpost (Pawn|GetController InteractingPawn))
    (:then
      (return (Outpost|IsOutpostOwner _asfort_player_controller_outpost)))
    (:CastFailed
      (return true))))

(fn CleanupWiggleMIDs
  (if (TouchPlatform|IsMobileGame)
    (Components|Timeline|Stop (Variables|Creative_Tiered_Chest|GetMobileOnInteractTL))
    (for _array_element (Variables|Default|GetMIDs)
      (Rendering|Material|SetScalarParameterValue _array_element "WiggleOnInteractAmount" 0.0 :!class /Script/Engine.MaterialInstanceDynamic))))

(fn CreateMobileMIDs
  (if (TouchPlatform|IsMobileGame)
    (for (_ _array_index) (Rendering|Material|GetMaterials (Variables|Debug|GetStaticMeshComponent))
      (bind _returnvalue (CallFunction|GetorCreateMIDOnDefaultMesh _array_index))
      (Utilities|Array|SetArrayElem (Variables|Default|GetMIDs) _array_index _returnvalue true))))

(fn GetMaxAudibleDistance
  (bind _chest_ambient_sound (Variables|Default|GetChest_Ambient_Sound))
  (bind _sound (Variables|/Script/Engine.AudioComponent|GetSound _chest_ambient_sound))
  (Utilities|IsValid _sound
    (:Is%20Valid
      (return (Variables|/Script/Engine.SoundBase|GetMaxDistance _sound)))
    (:Is%20Not%20Valid
      (return 0.0))))

(fn OnFreeResources
  (params (Resource%20Manager :type Creative%20Island%20Resource%20Manager%20Component%20Object%20Reference))
  (Utilities|IsValid Resource%20Manager
    (:Is%20Valid
      (Networking|CanExecuteCosmeticEvents
        (:True
          (Utilities|FlowControl|Sequence
            (:then_0
              (Class|CreativeIslandResourceManagerComponent|FreeResource (Variables|GetResource%20Manager) (Variables|Default|GetResourceLightTag) (Transformation|GetActorLocation)))
            (:then_1
              (Class|CreativeIslandResourceManagerComponent|FreeResource (Variables|GetResource%20Manager) (Variables|Default|GetResourceVFXTag) (Transformation|GetActorLocation)))
            (:then_2
              (Class|CreativeIslandResourceManagerComponent|FreeResource (Variables|GetResource%20Manager) (Variables|Default|GetResourceAudioTag) (Transformation|GetActorLocation)))))))))

(fn OnRequestResources
  (params (Resource%20Manager :type Creative%20Island%20Resource%20Manager%20Component%20Object%20Reference))
  (Utilities|IsValid Resource%20Manager
    (:Is%20Valid
      (Networking|CanExecuteCosmeticEvents
        (:True
          (Utilities|FlowControl|Sequence
            (:then_0
              (bind _returnvalue (Class|CreativeIslandResourceManagerComponent|RequestResource (Variables|GetResource%20Manager) (Variables|Default|GetResourceLightTag) (Transformation|GetActorLocation)))
              (Rendering|Components|Light|SetIntensity (Variables|Default|GetTrasureLight) (Math|Float|SelectFloat 1.7e+03 0.0 _returnvalue)))
            (:then_1
              (Class|BuildingFunctionLibrary|RequestComponentResource (Variables|GetResource%20Manager) (Variables|Default|GetResourceVFXTag) (Variables|Default|GetNS_Athena_Loot_Chest_Aura) (Transformation|GetActorLocation)))
            (:then_2
              (Class|BuildingFunctionLibrary|RequestComponentResource (Variables|GetResource%20Manager) (Variables|Default|GetResourceAudioTag) (Variables|Default|GetChest_Ambient_Sound) (Transformation|GetActorLocation)))))))))

(fn OnSetCustomDepthStencilValue
  (params (PrimComponents :type Array%20of%20Primitive%20Component%20Object%20References) (bUseCustomDepth :type Boolean) (StencilValue :type Integer))
  (for _array_element PrimComponents
    (Rendering|SetCustomDepthStencilValue _array_element (Variables|GetStencilValue))
    (if (Variables|GetbUseCustomDepth)
      (Rendering|SetRenderCustomDepth _array_element true)))
  (return true))

(fn SetLightVisibility
  (params (Visible :type Boolean))
  (Variables|Default|GetTrasureLight
    (:then
      (Rendering|SetVisibility (Variables|Default|GetTrasureLight) Visible false))
    (:else)))

(fn SetVisibleMobileInteractIcon
  (params (Visible :type Boolean))
  (bind (_returnvalue_roll _returnvalue_pitch _returnvalue_yaw) (Math|Rotator|BreakRotator (Transformation|GetActorRotation)))
  (if Visible
    ;[1 1 1] Create
    (bind _mobileinteracticon (Variables|Default|GetMobileInteractIcon)
      (:then
        (Rendering|SetVisibility _mobileinteracticon true false))
      (:else
        (Variables|Default|GetMobileInteractionMaterial
          (:then
            (bind _returnvalue (AddComponent|NODE_AddStaticMeshComponent-1 :RelativeTransform_Location (Variables|Default|GetMobileInteractIconLocation) :RelativeTransform_Rotation_Yaw (* _returnvalue_yaw -1.0) :RelativeTransform_Scale (Variables|Default|GetMobileInteractIconScale)))
            (bind _output_get (Variables|Default|SetMobileInteractIcon _returnvalue))
            (Rendering|Material|SetMaterial _output_get 0 (Variables|Default|GetMobileInteractionMaterial)))
          (:else))))
    (else
      ;[1 1 1] Destroy
      (bind _mobileinteracticon_2 (Variables|Default|GetMobileInteractIcon)
        (:then
          (Rendering|SetVisibility _mobileinteracticon_2 false false)
          (Components|DestroyComponent _mobileinteracticon_2))
        (:else)))))

(fn ShouldDie
  (params (Damage :type Float%20%28single-precision%29) (EventInstigator :type Controller%20Object%20Reference) (DamageCauser :type Actor%20Object%20Reference) (bIsTest :type Boolean) (EffectContext :type Gameplay%20Effect%20Context%20Handle%20Structure%20%28by%20ref%29))
  (bind _asfort_player_controller_outpost (Utilities|Casting|CastToFortPlayerControllerOutpost EventInstigator)
    (:then
      (return (Outpost|IsOutpostOwner _asfort_player_controller_outpost)))
    (:CastFailed
      (return true))))

(fn ConstructionScript
  (Networking|CanExecuteCosmeticEvents
    (:True
      (Utilities|EffectsQualityCheck 0
        (:Medium
          _)
        (:Low
          _))
      (Components|DestroyComponent (Variables|Default|GetTrasureLight)))))
