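; Blueprint for an Athena produce food box, a searchable loot container actor.
; On loot, remote clients play an interact-complete sound through the SoundLibrary component unless the searching pawn is locally controlled.
; It then animates the box by creating a dynamic material on its static mesh and driving the CODE_Timestamp parameter from a timeline.
; The construction script adds the TieredAmmoAthena material tag so level-placed instances receive the correct chrome material.
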(event EventOnLoot
  (|CallParentFunction)
  ; Play Open Sound
  (Networking|SwitchHasAuthority
    (:Authority
      _)
    (:Remote
      (bind _returnvalue (Container|GetSearchingPawn))
      (Utilities|IsValid _returnvalue
        (:Is%20Valid
          (if (not (Pawn|IsLocallyControlled _returnvalue))
            (SoundLibrary|PlayEvent (Variables|Default|GetSoundLibrary%20Component) "(TagName=\"SoundLibrary.Item.Interact.Complete\")"))))))
  ;[1 1 1] Animate Chest
  (bind _returnvalue_2 (Rendering|Material|CreateDynamicMaterialInstance (Variables|Debug|GetStaticMeshComponent) 0 0 "None" :!class /Script/Engine.PrimitiveComponent))
  (bind (_direction _time) (Timeline|Timeline_0 Play)
    (:Update
      (Rendering|Material|SetScalarParameterValue _returnvalue_2 "CODE_Timestamp" _time :!class /Script/Engine.MaterialInstanceDynamic))))

(fn ConstructionScript
  ;[0.32 0.9402 1] Make sure actors that are already placed in levels get this tag so that they get the right chrome material added to them.
  (|CallParentFunction)
  (Utilities|Array|AddUnique (Variables|Actor|GetTags) "Lackluster.Material.TieredAmmoAthena"))
