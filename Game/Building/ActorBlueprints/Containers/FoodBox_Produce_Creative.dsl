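; Blueprint for the Creative produce food box container, which animates its mesh through its material when looted.
; On loot it runs the parent logic, creates a dynamic material instance on the static mesh's first slot, then plays a timeline that writes its value into the CODE_Timestamp scalar parameter each update.
; It reports itself as targetable by linear movers and disallows invisible props.
; The construction script is empty.
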(event EventOnLoot
  ;[1 1 1] Animate Chest
  (|CallParentFunction)
  (bind _returnvalue (Rendering|Material|CreateDynamicMaterialInstance (Variables|Debug|GetStaticMeshComponent) 0 0 "None" :!class /Script/Engine.PrimitiveComponent))
  (bind (_direction _time) (Timeline|Timeline_0 Play)
    (:Update
      (Rendering|Material|SetScalarParameterValue _returnvalue "CODE_Timestamp" _time :!class /Script/Engine.MaterialInstanceDynamic))))

(fn IsTargetableByLinearMover
  (return true))

(fn ShouldInvisiblePropsBeAllowed
  (return false))

(fn ConstructionScript)
