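; Blueprint for the Mailbox_01a building container, a lootable mailbox prop.
; On the loot event it activates its Effect_Mail component without a reset, presumably playing a mail effect when searched.
; The construction script is empty.
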(event EventOnLoot
  (Components|Activation|Activate (Variables|Mailbox_01a|GetEffect_Mail) false))

(fn ConstructionScript)
