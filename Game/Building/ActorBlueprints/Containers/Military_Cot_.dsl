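; Military cot container actor Blueprint that supplies the bed-specific foley sounds for loot searching.
; The OnBeginSearch event plays the bed search cue at the actor's location.
; The OnLootRepeat event plays the bed search-end cue at the same location.
; The construction script is empty, and no search or loot logic is defined here.
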(event Custom|OnBeginSearch
  (Audio|PlaySoundatLocation "/Game/Sounds/Foley_Loot/Containers/Bed/Bed_Search_Cue.Bed_Search_Cue" (Transformation|GetActorLocation)))

(event Custom|OnLootRepeat
  (Audio|PlaySoundatLocation "/Game/Sounds/Foley_Loot/Containers/Bed/Bed_SearchEnd_Cue.Bed_SearchEnd_Cue" (Transformation|GetActorLocation)))

(fn ConstructionScript)
