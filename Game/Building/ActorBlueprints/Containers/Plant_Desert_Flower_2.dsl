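; Blueprint for a lootable desert flower plant, placed among the game's building container actors.
; When looted, it activates the effect component obtained from its LCD_DES_Shrub_04 variable, without resetting it.
; The construction script only calls the parent version, so other container behaviour is inherited from the parent class.
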(event EventOnLoot
  (Components|Activation|Activate (Variables|LCD_DES_Shrub_04|GetEffect) false))

(fn ConstructionScript
  (|CallParentFunction))
