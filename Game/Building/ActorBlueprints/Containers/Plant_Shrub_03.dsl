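; Shrub container Blueprint (Plant_Shrub_03) that sometimes spawns with an extra dust particle effect.
; On BeginPlay, a random roll against PercentChanceOfAddedFX decides whether to add a manually attached static mesh component at the root's world transform, stored as Dust_Mesh_Particles.
; On destruction, it destroys that dust component.
; The construction script only calls its parent.
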(event EventBeginPlay
  ;[1 0.7631 0] Creates Dust Particles
  (if (< (Math|Random|RandomFloatinRange 0.0 100) (Variables|Default|GetPercentChanceOfAddedFX))
    (bind _returnvalue (AddComponent|NODE_AddStaticMeshComponent-0 :bManualAttachment true :RelativeTransform (Transformation|GetWorldTransform (Variables|Transformation|GetRootComponent))))
    (Variables|Default|SetDust_Mesh_Particles _returnvalue)))

(event EventDestroyed
  ;[1 0.7631 0] Creates Dust Particles
  (Components|DestroyComponent (Variables|Default|GetDust_Mesh_Particles)))

(fn ConstructionScript
  (|CallParentFunction))
