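; Parent Blueprint for lootable shrub plant containers, adding visual and audio feedback when the plant is looted.
; On loot, it runs the inherited logic, then continues only where cosmetic events can execute.
; If a looted particle system is set, it spawns that effect and plays a fixed flower-collect sound at the actor's location, so without one the sound is also skipped.
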(event EventOnLoot
  (bind _returnvalue (Transformation|GetActorLocation))
  (|CallParentFunction)
  (Networking|CanExecuteCosmeticEvents
    (:True
      (bind _lootedparticlesystem (Variables|Default|GetLootedParticleSystem)
        (:then
          (Effects|SpawnEmitteratLocation _lootedparticlesystem _returnvalue)
          (Audio|PlaySoundatLocation "/Game/Sounds/Foley_Loot/Ingredients/Health/Flowers_Collect_Cue.Flowers_Collect_Cue" _returnvalue))
        (:else)))))

(fn ConstructionScript
  (|CallParentFunction))
