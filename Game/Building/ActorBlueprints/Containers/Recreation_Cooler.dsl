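; Recreation_Cooler is a searchable cooler container Blueprint that handles the sound played when a pawn loots it.
; On loot, a valid, locally viewed searching pawn triggers PlaySoundforAll, whose event here only calls the parent.
; Any other valid pawn gets the cooler-open cue played at the actor's location.
; The construction script runs the parent, then adds the Lackluster.Material.TieredAmmoAthena tag to the actor's tags.
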(event EventOnLoot
  (bind _returnvalue (Container|GetSearchingPawn))
  (Utilities|IsValid _returnvalue
    (:Is%20Valid
      (if (Pawn|IsLocallyViewed _returnvalue)
        (CallFunction|PlaySoundforAll)
        (else
          (Audio|PlaySoundatLocation "/Game/Sounds/Foley_Loot/Containers/Cooler/Cooler_Open_Cue.Cooler_Open_Cue" (Transformation|GetActorLocation)))))))

(event Custom|Play%20Sound%20For%20All
  (|CallParentFunction))

(fn ConstructionScript
  (|CallParentFunction)
  (Utilities|Array|AddUnique (Variables|Actor|GetTags) "Lackluster.Material.TieredAmmoAthena"))
