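; Starlight secret room chest container Blueprint that shares its search loot with the other players when opened.
; On the mission chest-opened event, it loops over all Fort player controllers and skips the pawn that searched it.
; Each other player receives mission rewards as pickups around the chest's location, drawn from the chest's own search loot tier group.
; The construction script is empty.
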(event Custom|MissionNotifyOnChestOpened
  (params (SearchedContainer :type Building%20Container%20Object%20Reference) (SearcherPawn :type Fort%20Pawn%20Object%20Reference))
  (for _array_element (Fortnite|GetAllFortPlayerControllers true)
    (if (Utilities|NotEqual(Object) SearcherPawn (Pawn|GetControlledPawn _array_element))
      (MissionHelpers|GiveMissionRewardstoPlayerasPickups :PlayerController _array_element :StartPosition (Transformation|GetActorLocation) :Radius 5e+01 :LootTierGroup (Variables|Container|GetSearchLootTierGroup) :LootTier 0))))

(fn ConstructionScript)
