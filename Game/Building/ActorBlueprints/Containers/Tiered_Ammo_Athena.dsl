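; Blueprint for the tiered ammo loot container, handling its cosmetics and events when highlighted, searched and looted.
; A timeline scales the mobile interact icon in on highlight and reverses on unhighlight or search.
; Looting spawns open FX, animates a material timestamp at higher detail, plays sound once for non-locally-viewed searchers, and sends an authority-side open event inside minigames.
; Construction adds a tag for chrome material assignment.
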(event Container|EventEnableBacchusHighlight
  (bind (_returnvalue_roll _returnvalue_pitch _returnvalue_yaw) (Math|Rotator|BreakRotator (Transformation|GetActorRotation)))
  ;[1 1 1] Mobile highlighting
  (Networking|CanExecuteCosmeticEvents
    (:True
      (Utilities|SetVisibleMobileIcon true (Variables|Default|GetMobileInteractIcon) (Variables|Default|GetMobileInteractionMaterial) self (Variables|Default|GetMobileInteractIconLocation) (Math|Rotator|MakeRotator 0.0 0.0 (* _returnvalue_yaw -1.0)) (Variables|Default|GetMobileInteractIconScale)
        (:Out
          (bind (_direction _lerpobject _lerpinteactoicon) (Timeline|MobileSelectedTL Play)
            (:Update
              (bind _mobileinteracticon (Variables|Default|GetMobileInteractIcon)
                (:then
                  (Transformation|SetRelativeScale3D _mobileinteracticon (Math|Vector|Lerp(Vector) "0, 0, 0" (Variables|Default|GetMobileInteractIconScale) _lerpinteactoicon)))
                (:else)))
            (:Finished
              (switch Utilities|FlowControl|Switch|SwitchonETimelineDirection _direction
                (:Backward
                  (Utilities|SetVisibleMobileIcon :Mesh (Variables|Default|GetMobileInteractIcon) :Rotation (0.0 0.0 0.0)))))))))))

(event Container|EventDisableBacchusHighlight
  ;[1 1 1] Mobile highlighting
  (if (TouchPlatform|IsMobileGame)
    (Networking|CanExecuteCosmeticEvents
      (:True
        (Timeline|MobileSelectedTL Reverse)))))

(event EventOnBeginSearch
  ;[1 1 1] Mobile highlighting
  (if (Variables|Container|GetbAlreadySearched)
    _
    (elif (TouchPlatform|IsMobileGame)
      (Networking|CanExecuteCosmeticEvents
        (:True
          (Timeline|MobileSelectedTL Reverse))))))

(event EventOnLoot
  (bind _fx_open (Variables|FX|GetFX_Open))
  (Utilities|FlowControl|Sequence
    (:then_0
      (Networking|CanExecuteCosmeticEvents
        (:True
          (Utilities|FlowControl|Sequence
            (:then_0
              (CallFunction|AudioonOpened))
            (:then_1
              ;[1 1 1] Play Open FX if valid
              (Utilities|IsValid _fx_open
                (:Is%20Valid
                  (Effects|SpawnEmitteratLocation _fx_open (+ (Transformation|GetActorLocation) (Variables|FX|GetFX_OpenOffsetLocation)) (Math|Rotator|MakeRotator (.roll (Transformation|GetActorRotation)) (+ (.pitch (Transformation|GetActorRotation)) (Variables|FX|GetFX_OpenPitchOffset)) (.yaw (Transformation|GetActorRotation)))))))
            (:then_2
              ;[1 1 1] Animate Chest
              (bind _returnvalue (Rendering|Material|CreateDynamicMaterialInstance (Variables|Debug|GetStaticMeshComponent) 0 0 "None" :!class /Script/Engine.PrimitiveComponent))
              (if (>= (Rendering|GetRenderingDetailMode) 3)
                (bind (_direction _time) (Timeline|Timeline_0 PlayFromStart)
                  (:Update
                    (Rendering|Material|SetScalarParameterValue _returnvalue "CODE_Timestamp" _time :!class /Script/Engine.MaterialInstanceDynamic)))
                (else
                  (Rendering|Material|SetScalarParameterValue _returnvalue "CODE_Timestamp" 1.0 :!class /Script/Engine.MaterialInstanceDynamic))))))))
    (:then_1
      ;[0 1 0.9218 0.25] Check if we are in Creative Royale and trigger Open Event
      (Networking|SwitchHasAuthority
        (:Authority
          (Utilities|IsValid (Minigame|GetMinigame self :!class /Script/FortniteGame.MinigameBlueprintLibrary)
            (:Is%20Valid
              (Class|LootControllerFunctionLibrary|MakeandSendOpenContainerEvent self (Variables|Default|GetContainer%20Type)))))))))

(event Custom|AudioOnOpened
  ;[0.2561 0.2629 0.3448] Play Open Sound
  ;;[0.01266 0.1001 0.15 0.5 12] Stops Sound Double for 1p
  (bind _returnvalue (Container|GetSearchingPawn))
  (Utilities|IsValid _returnvalue
    (:Is%20Valid
      (if (Pawn|IsLocallyViewed _returnvalue)
        _
        (elif (Variables|Default|GetbOpened)
          _
          (else
            (SoundLibrary|PlaySoundSimple (WorldSubsystems|GetSoundLibrarySubsystem) self (Variables|Default|GetSoundLib_OpenEvent) :!class /Script/SoundLibrary.SoundLibrarySubsystem)
            (Variables|Default|SetbOpened true)))))))

(fn IsTargetableByLinearMover
  (return true))

(fn SetVisibleMobileInteractIcon
  (params (Visible :type Boolean))
  (bind (_returnvalue_roll _returnvalue_pitch _returnvalue_yaw) (Math|Rotator|BreakRotator (Transformation|GetActorRotation)))
  (if Visible
    ;[1 1 1] Create
    (bind _mobileinteracticon (Variables|Default|GetMobileInteractIcon)
      (:then
        (Rendering|SetVisibility _mobileinteracticon true false))
      (:else
        (Variables|Default|GetMobileInteractionMaterial
          (:then
            (bind _returnvalue (AddComponent|NODE_AddStaticMeshComponent-0 :RelativeTransform_Location (Variables|Default|GetMobileInteractIconLocation) :RelativeTransform_Rotation_Yaw (* _returnvalue_yaw -1.0) :RelativeTransform_Scale (Variables|Default|GetMobileInteractIconScale)))
            (bind _output_get (Variables|Default|SetMobileInteractIcon _returnvalue))
            (Rendering|Material|SetMaterial _output_get 0 (Variables|Default|GetMobileInteractionMaterial)))
          (:else))))
    (else
      ;[1 1 1] Destroy
      (bind _mobileinteracticon_2 (Variables|Default|GetMobileInteractIcon)
        (:then
          (Rendering|SetVisibility _mobileinteracticon_2 false false)
          (Components|DestroyComponent _mobileinteracticon_2))
        (:else)))))

(fn ShouldInvisiblePropsBeAllowed
  (return false))

(fn ConstructionScript
  ;[0.32 0.9402 1] Make sure actors that are already placed in levels get this tag so that they get the right chrome material added to them.
  (|CallParentFunction)
  (Utilities|Array|AddUnique (Variables|Actor|GetTags) "Lackluster.Material.TieredAmmoAthena"))
