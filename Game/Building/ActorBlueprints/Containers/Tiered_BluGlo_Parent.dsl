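; Parent Blueprint for tiered BluGlo containers, placed among the building container actors.
; On BeginPlay it gets its static mesh component and, if valid, hides it in game without propagating to children.
; The construction script is empty.
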(event EventBeginPlay
  (bind _staticmeshcomponent (Variables|Debug|GetStaticMeshComponent))
  (Utilities|IsValid _staticmeshcomponent
    (:Is%20Valid
      (Development|SetHiddeninGame _staticmeshcomponent true false))))

(fn ConstructionScript)
