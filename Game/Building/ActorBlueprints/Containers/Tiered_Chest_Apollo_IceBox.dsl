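; Ice box loot container Blueprint, a tiered chest variant from the Apollo content set.
; On loot, it runs the parent logic, then sets the loot effect particle component's template to the ice machine opening effect.
; The construction script adds a material actor tag without duplicates, so already placed instances receive the correct chrome material.
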(event EventBeginPlay
  (|CallParentFunction))

(event EventOnLoot
  (|CallParentFunction)
  (Effects|Components|ParticleSystem|SetTemplate (Variables|Default|GetLoot_Effect) "/Game/Effects/Fort_Effects/World/Chests/P_Ice_Machine_Open.P_Ice_Machine_Open"))

(fn ConstructionScript
  ;[0.32 0.9402 1] Make sure actors that are already placed in levels get this tag so that they get the right chrome material added to them.
  (|CallParentFunction)
  (Utilities|Array|AddUnique (Variables|Actor|GetTags) "Lackluster.Material.TieredChestAthena"))
