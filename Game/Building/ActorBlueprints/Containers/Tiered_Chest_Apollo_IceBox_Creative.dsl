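; Creative-mode variant of the Apollo Ice Box tiered chest, a lootable container Blueprint.
; Its loot event runs the parent logic, then sets the Loot_Effect particle component's template to the ice machine opening effect.
; BeginPlay and the construction script only call the parent, so all other behavior is inherited.
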(event EventBeginPlay
  (|CallParentFunction))

(event EventOnLoot
  (|CallParentFunction)
  (Effects|Components|ParticleSystem|SetTemplate (Variables|Default|GetLoot_Effect) "/Game/Effects/Fort_Effects/World/Chests/P_Ice_Machine_Open.P_Ice_Machine_Open"))

(fn ConstructionScript
  (|CallParentFunction))
