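; Tiered Battle Royale loot chest container with glow, light, chime and opening cosmetics.
; BeginPlay sets up effects, swapping to blue rare-chest VFX when the mesh matches, also rechecked on mesh-set change.
; Looting spawns rarity-tinted burst FX and drives a timeline-fed material timestamp for opening, reversible on loot reset.
; Cosmetic logic is gated from gameplay, which sends minigame open events; mobile adds MID wiggle, outline and interact icon.
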(event EventBeginPlay
  ;[0.4993 1 0.5586] Begin Play
  (Networking|CanExecuteCosmeticEvents
    (:True
      (CallFunction|FXGlow&SmokeSheetSetup)
      (CallFunction|CreateMobileMIDs)
      (if (Variables|Container|GetbAlreadySearched)
        (CallFunction|SetLightVisibility false)
        (Audio|Components|Audio|Stop (Variables|Default|GetChest_Ambient_Sound))
        (Variables|Default|SetbOpened true)
        (else
          (Rendering|SetVisibility (Variables|Default|GetFX%20Glow%20&%20Smoke%20Sheet) true false)
          (CallFunction|SetLightVisibility true)
          (CallFunction|SwapChestVFX)
          (Utilities|FlowControl|Sequence
            (:then_0
              (if (Variables|Default|GetEnableAudioIndicator)
                (CallFunction|StartAudiovisualIndicator (Variables|Default|GetChest_Ambient_Sound) "Chest" (Variables|Default|GetChestChimeVisualUpdate))))
            (:then_1
              (Building|BindEventtoOnMeshSetChanged (EventDispatchers|CreateEvent :!event Custom|MeshSetChanged)))))))))

(event Custom|Mobile_OnInteract
  ;[1 1 1] Mobile Wiggle on Interact
  (Networking|CanExecuteCosmeticEvents
    (:True
      (bind (_direction _lerp) (Timeline|MobileOnInteractTL Play)
        (:Update
          (for _array_element (Variables|Default|GetMIDs)
            (Rendering|Material|SetScalarParameterValue _array_element "WiggleOnInteractAmount" (Math|Float|Lerp 0.0 (Variables|Default|GetMobileWiggleAmount) _lerp) :!class /Script/Engine.MaterialInstanceDynamic)))))))

(event Custom|Mobile_Outline_On
  ;[1 1 1] Mobile Outline
  (for _array_element (Variables|Default|GetMIDs)
    (Rendering|Material|SetScalarParameterValue _array_element "Outline_Control" 1.0 :!class /Script/Engine.MaterialInstanceDynamic)
    (Rendering|Material|SetScalarParameterValue _array_element "Pulse_Control" 1.0 :!class /Script/Engine.MaterialInstanceDynamic)
    (Rendering|Material|SetVectorParameterValue _array_element "Color" "(R=1.000000,G=1.000000,B=1.000000,A=1.000000)" :!class /Script/Engine.MaterialInstanceDynamic)))

(event Container|EventEnableBacchusHighlight
  ;[1 1 1] Mobile highlighting
  (if (not (Variables|Container|GetbAlreadySearched))
    (Networking|CanExecuteCosmeticEvents
      (:True
        (CallFunction|SetVisibleMobileInteractIcon true)
        (bind (_direction _lerpobject _lerpinteactoicon) (Timeline|MobileSelectedTL Play)
          (:Update
            (bind _mobileinteracticon (Variables|Default|GetMobileInteractIcon)
              (:then
                (Transformation|SetRelativeScale3D _mobileinteracticon (Math|Vector|Lerp(Vector) "0, 0, 0" (Variables|Default|GetMobileInteractIconScale) _lerpinteactoicon)))
              (:else)))
          (:Finished
            (switch Utilities|FlowControl|Switch|SwitchonETimelineDirection _direction
              (:Backward
                (CallFunction|SetVisibleMobileInteractIcon false)))))))))

(event Custom|Mobile_OnDisconnect
  ;[1 1 1] Mobile Wiggle on Interact
  (Networking|CanExecuteCosmeticEvents
    (:True
      (Timeline|MobileOnInteractTL Reverse))))

(event Custom|Mobile_Outline_Off
  ;[1 1 1] Mobile Outline
  (for _array_element (Variables|Default|GetMIDs)
    (Rendering|Material|SetScalarParameterValue _array_element "Outline_Control" 0.0 :!class /Script/Engine.MaterialInstanceDynamic)
    (Rendering|Material|SetScalarParameterValue _array_element "Pulse_Control" 0.0 :!class /Script/Engine.MaterialInstanceDynamic)))

(event Container|EventDisableBacchusHighlight
  ;[1 1 1] Mobile highlighting
  (if (TouchPlatform|IsMobileGame)
    (Networking|CanExecuteCosmeticEvents
      (:True
        (Timeline|MobileSelectedTL Reverse)))))

(event EventOnBeginSearch
  ;[1 1 1] Mobile highlighting
  (if (Variables|Container|GetbAlreadySearched)
    _
    (elif (TouchPlatform|IsMobileGame)
      (Networking|CanExecuteCosmeticEvents
        (:True
          (Timeline|MobileSelectedTL Reverse))))))

(event EventOnSetSearched
  (Networking|CanExecuteCosmeticEvents
    (:True
      (CallFunction|HideChestOpenableCosmetics))
    (:False
      (Variables|Building|SetbDestroyOnPlayerBuildingPlacement true) ; Allows players to build over the opened chest mesh (but not when there is loot in there)
    )))

(event EventDestroyed
  (Networking|CanExecuteCosmeticEvents
    (:True
      (CallFunction|HideChestOpenableCosmetics))
    (:False
      (Variables|Building|SetbDestroyOnPlayerBuildingPlacement true) ; Allows players to build over the opened chest mesh (but not when there is loot in there)
    )))

(event EventOnLoot
  (Networking|CanExecuteCosmeticEvents
    (:True
      (Utilities|FlowControl|Sequence
        (:then_0
          (CallFunction|HideChestOpenableCosmetics))
        (:then_1
          (CallFunction|DetermineVFXRotation))
        (:then_2
          (CallFunction|SpawnChestOpenVFX))
        (:then_3
          (CallFunction|PlayChestOpeningTimeline))))
    (:False
      (Variables|Building|SetbDestroyOnPlayerBuildingPlacement true) ; Allows players to build over the opened chest mesh (but not when there is loot in there)
      (Default|CallContainerOpened)
      (Utilities|IsValid (Minigame|GetMinigame self :!class /Script/FortniteGame.MinigameBlueprintLibrary)
        (:Is%20Valid
          (Networking|SwitchHasAuthority
            (:Authority
              (bind _container_type (CallFunction|GetChestType))
              (Class|LootControllerFunctionLibrary|MakeandSendOpenContainerEvent self _container_type))))))))

(event Custom|HideChestOpenableCosmetics
  (bind _fx_glow___smoke_sheet (Variables|Default|GetFX%20Glow%20&%20Smoke%20Sheet))
  (Utilities|FlowControl|DoOnce false
    (:Completed
      (Components|Activation|Deactivate (Variables|Default|GetRare%20ChestFX))
      (if (not (Variables|Default|GetbCanBeReset))
        (Components|DestroyComponent _fx_glow___smoke_sheet)
        (else
          (Rendering|SetVisibility _fx_glow___smoke_sheet false false)))
      (CallFunction|SetLightVisibility false)
      (bind _returnvalue (Container|GetSearchingPawn))
      ;[1 1 1] Prevents double triggering of opening sound for the local player
      (Utilities|IsValid _returnvalue
        (:Is%20Valid
          (if (Pawn|IsLocallyControlled _returnvalue)
            (bind _chest_ambient_sound (Variables|Default|GetChest_Ambient_Sound :!label _label)
              (:then
                (Audio|Components|Audio|FadeOut _chest_ambient_sound 0.5 0.0 "Linear" :!class /Script/Engine.AudioComponent)
                (Variables|Default|SetbOpened true))
              (:else))
            (else
              (CallFunction|AudioonOpened)))
          (CallFunction|StopAudiovisualIndicator)
          (CallFunction|CleanupWiggleMIDs)
          (Components|Timeline|Stop (Variables|Tiered_Chest_Athena|GetMobileSelectedTL))
          (CallFunction|SetVisibleMobileInteractIcon false)
          (Rendering|SetRenderCustomDepth (Variables|Debug|GetStaticMeshComponent) false)
          (Building|BroadcastCustomRenderDepthChange false)
          (CallFunction|DestroyAdditionalVFXComponents)
          ;[1 1 1] Prevents sound library tags to get applied after the chest is opened
          (if (Variables|Default|GetbOpened)
            (SoundLibrary|DisableEventsforActor (WorldSubsystems|GetSoundLibrarySubsystem) self)))
        (:Is%20Not%20Valid
          _))
      (goto _label))))

(event Custom|ResetOpenableCosmetics
  (bind _fx_glow___smoke_sheet (Variables|Default|GetFX%20Glow%20&%20Smoke%20Sheet))
  (Utilities|FlowControl|DoOnce false
    (:Completed
      (Components|Activation|Deactivate (Variables|Default|GetRare%20ChestFX))
      (if (not (Variables|Default|GetbCanBeReset))
        (Components|DestroyComponent _fx_glow___smoke_sheet)
        (else
          (Rendering|SetVisibility _fx_glow___smoke_sheet false false)))
      (CallFunction|SetLightVisibility false)
      (bind _returnvalue (Container|GetSearchingPawn))
      ;[1 1 1] Prevents double triggering of opening sound for the local player
      (Utilities|IsValid _returnvalue
        (:Is%20Valid
          (if (Pawn|IsLocallyControlled _returnvalue)
            (bind _chest_ambient_sound (Variables|Default|GetChest_Ambient_Sound :!label _label)
              (:then
                (Audio|Components|Audio|FadeOut _chest_ambient_sound 0.5 0.0 "Linear" :!class /Script/Engine.AudioComponent)
                (Variables|Default|SetbOpened true))
              (:else))
            (else
              (CallFunction|AudioonOpened)))
          (CallFunction|StopAudiovisualIndicator)
          (CallFunction|CleanupWiggleMIDs)
          (Components|Timeline|Stop (Variables|Tiered_Chest_Athena|GetMobileSelectedTL))
          (CallFunction|SetVisibleMobileInteractIcon false)
          (Rendering|SetRenderCustomDepth (Variables|Debug|GetStaticMeshComponent) false)
          (Building|BroadcastCustomRenderDepthChange false)
          (CallFunction|DestroyAdditionalVFXComponents)
          ;[1 1 1] Prevents sound library tags to get applied after the chest is opened
          (if (Variables|Default|GetbOpened)
            (SoundLibrary|DisableEventsforActor (WorldSubsystems|GetSoundLibrarySubsystem) self)))
        (:Is%20Not%20Valid
          _))
      (goto _label))))

(event Custom|SpawnChestOpenVFX
  (bind _p_loot_chest_opened_fx_ns (Variables|Default|GetP_Loot_Chest_Opened_FX_NS))
  (bind _looteffect_ns (Variables|Default|GetLootEffect_NS))
  (bind _loot_effect (Variables|Default|GetLoot_Effect))
  (bind _p_loot_chest_opened_fx (Variables|Default|GetP_Loot_Chest_Opened_FX))
  (Utilities|FlowControl|Sequence
    (:then_0
      ;[1 1 1] Adding for NS Yi.Xie 10.31
      (Utilities|IsValid _p_loot_chest_opened_fx_ns
        (:Is%20Valid
          (bind _returnvalue (Niagara|SpawnSystematLocation _p_loot_chest_opened_fx_ns (Transformation|GetActorLocation) (Transformation|GetActorRotation)))
          (Variables|Default|SetLootEffect_NS _returnvalue))))
    (:then_1
      (Utilities|IsValid _p_loot_chest_opened_fx
        (:Is%20Valid
          (bind _returnvalue_2 (Effects|SpawnEmitteratLocation _p_loot_chest_opened_fx (Transformation|GetActorLocation) (Math|Rotator|MakeRotator (.roll (Transformation|GetActorRotation)) (+ (.pitch (Transformation|GetActorRotation)) (Variables|Default|GetOpenFX_Rotation)) (.yaw (Transformation|GetActorRotation)))))
          (Variables|Default|SetLoot_Effect _returnvalue_2))))
    (:then_2
      (bind _returnvalue_3 (Fortnite|GetRarityData (Variables|Container|GetHighestRarity)))
      (Utilities|FlowControl|Sequence
        (:then_0
          (Utilities|IsValid _loot_effect
            (:Is%20Valid
              (Effects|Components|ParticleSystem|SetVectorParameter _loot_effect "color" (Math|Conversions|ToVector(LinearColor) (.Color2 (Utilities|Struct|BreakFortRarityItemData _returnvalue_3)))))))
        (:then_1
          (Utilities|IsValid _looteffect_ns
            (:Is%20Valid
              (Effects|Components|ParticleSystem|SetVectorParameter _looteffect_ns "color" (Math|Conversions|ToVector(LinearColor) (.Color2 (Utilities|Struct|BreakFortRarityItemData _returnvalue_3)))))))))))

(event Custom|DetermineVFXRotation
  ;[0.926 0 1 1 80] FOR GALILEO ONLY
  (bind _returnvalue (Container|GetLootTier))
  (switch int _returnvalue
    (:Default
      (Variables|Default|SetOpenFX_Rotation -9e+01) ; Default chest
    )
    (:3
      (Variables|Default|SetP_Loot_Chest_Opened_FX (Variables|Default|GetGalileo_Chest_Open_FX)) ; 3 == Galileo Chest
      (Variables|Default|SetOpenFX_Rotation 0.0))))

(event EventOnLootReset
  (Networking|CanExecuteCosmeticEvents
    (:True
      (CallFunction|ResetOpeningTimeline))))

(event Custom|OnClosingComplete
  (Utilities|FlowControl|Sequence
    (:then_0
      (CallFunction|ResetOpenableCosmetics))
    (:then_1
      (CallFunction|SetLightVisibility true)
      (bind _chest_ambient_sound (Variables|Default|GetChest_Ambient_Sound)
        (:then
          (Audio|Components|Audio|FadeIn _chest_ambient_sound 0.0 1.0 0.0 "Linear" :!class /Script/Engine.AudioComponent)
          (CallFunction|StartAudiovisualIndicator (Variables|Default|GetChest_Ambient_Sound) "Chest" (Variables|Default|GetChestChimeVisualUpdate)))
        (:else)))
    (:then_2
      (CallFunction|FXGlow&SmokeSheetSetup))
    (:then_3
      (Rendering|SetVisibility (Variables|Default|GetFX%20Glow%20&%20Smoke%20Sheet) true false)
      (CallFunction|SwapChestVFX))))

(event Custom|PlayChestOpeningTimeline
  (Utilities|FlowControl|DoOnce false
    (:Completed
      (bind (_direction _time) (Timeline|Timeline_0 Play)
        (:Update
          (Rendering|Material|SetScalarParameterValueonMaterials (Variables|Debug|GetStaticMeshComponent) "CODE_Timestamp" _time))
        (:Finished
          (if (Variables|Default|GetbClosing)
            (Variables|Default|SetbClosing false)
            (CallFunction|OnClosingComplete)))))))

(event Custom|ResetOpeningTimeline
  (Utilities|FlowControl|Sequence
    (:then_0
      (Utilities|FlowControl|DoOnce false
        (:Completed
          (Timeline|Timeline_0 Play))))
    (:then_1
      (Timeline|Timeline_0 ReverseFromEnd))
    (:then_2
      (Variables|Default|SetbClosing true))))

(event EventOnDeathPlayEffects
  (params (Damage :type Float%20%28single-precision%29) (DamageTags :type Gameplay%20Tag%20Container%20Structure%20%28by%20ref%29) (Momentum :type Vector :default "0, 0, 0") (HitInfo :type Hit%20Result%20Structure%20%28by%20ref%29) (InstigatedBy :type Fort%20Pawn%20Object%20Reference) (DamageCauser :type Actor%20Object%20Reference) (EffectContext :type Gameplay%20Effect%20Context%20Handle%20Structure))
  (CallFunction|AudioonOpened))

(event Custom|AudioOnOpened
  (if (not (Variables|Default|GetbOpened))
    (Audio|Components|Audio|FadeOut (Variables|Default|GetChest_Ambient_Sound) 0.5 0.0 "Linear" :!class /Script/Engine.AudioComponent)
    (SoundLibrary|PlaySoundSimple (WorldSubsystems|GetSoundLibrarySubsystem) self (Variables|Default|GetSoundLib_OpenEvent) :!class /Script/SoundLibrary.SoundLibrarySubsystem)
    (bind _open_chest_sound (Variables|Sound|GetOpen%20Chest%20Sound)
      (:then
        (bind _chestopensourceeffectoverride (Variables|Sound|GetChestOpenSourceEffectOverride)
          (:then
            (Variables|/Script/Engine.SoundBase|SetSourceEffectChain _chestopensourceeffectoverride _open_chest_sound))
          (:else
            (Variables|/Script/Engine.SoundBase|SetSourceEffectChain 0 _open_chest_sound)))
        (Audio|PlaySoundatLocation (Variables|Sound|GetOpen%20Chest%20Sound) (Transformation|GetActorLocation)))
      (:else
        _))
    (Variables|Default|SetbOpened true)))

(event Custom|SwapChestVFX
  (bind _treasurelight (Variables|Default|GetTreasureLight))
  (bind _fx_glow___smoke_sheet (Variables|Default|GetFX%20Glow%20&%20Smoke%20Sheet))
  ;[1 1 1] Condition
  (if (CallFunction|IsRareChest)
    (if (not (Utilities|IsValid (Variables|Default|GetRare%20ChestFX) :!class /Script/Engine.KismetSystemLibrary))
      (bind _returnvalue (Niagara|SpawnSystemAttached "/Game/Effects/Fort_Effects/World/NewChests/NS_Chest_Rare.NS_Chest_Rare" (Variables|Debug|GetStaticMeshComponent) "None" "0, 0, 0" "0, 0, 0" "KeepRelativeOffset" true))
      (Variables|Default|SetRare%20ChestFX _returnvalue))
    (Rendering|Material|SetScalarParameterValueonMaterials _fx_glow___smoke_sheet "Corona_Opacity" 2e+01)
    (Rendering|Material|SetScalarParameterValueonMaterials _fx_glow___smoke_sheet "Glow_Opacity" 7.0)
    (Rendering|Material|SetVectorParameterValueonMaterials _fx_glow___smoke_sheet "Corona_Colour" "0.450000,0.874998,3.000000")
    (Rendering|Material|SetVectorParameterValueonMaterials _fx_glow___smoke_sheet "Glow_Colour" "0.050000,0.240000,1.000000")
    (Rendering|Components|Light|SetLightColor _treasurelight "(R=0.099899,G=0.270498,B=1.000000,A=1.000000)" true)
    (Rendering|Components|Light|SetIntensity _treasurelight 5.0)
    (Transformation|SetRelativeLocation _treasurelight "0.000000,62.000000,76.000000" false false)
    (Variables|Default|SetP_Loot_Chest_Opened_FX_NS "/Game/Effects/Fort_Effects/World/NewChests/NS_ChestBurst_Rare.NS_ChestBurst_Rare")
    (Variables|Default|SetP_Loot_Chest_Opened_FX 0)
    (CallFunction|FXGlow&SmokeSheetSetup)))

(event Custom|MeshSetChanged
  (params (BuildingSMActor :type Building%20Static%20Mesh%20Object%20Reference) (NewMeshSet :type Mesh%20Set%20Structure%20%28by%20ref%29))
  (bind _treasurelight (Variables|Default|GetTreasureLight))
  (bind _fx_glow___smoke_sheet (Variables|Default|GetFX%20Glow%20&%20Smoke%20Sheet))
  ;[1 1 1] Condition
  (if (CallFunction|IsRareChest)
    (if (not (Utilities|IsValid (Variables|Default|GetRare%20ChestFX) :!class /Script/Engine.KismetSystemLibrary))
      (bind _returnvalue (Niagara|SpawnSystemAttached "/Game/Effects/Fort_Effects/World/NewChests/NS_Chest_Rare.NS_Chest_Rare" (Variables|Debug|GetStaticMeshComponent) "None" "0, 0, 0" "0, 0, 0" "KeepRelativeOffset" true))
      (Variables|Default|SetRare%20ChestFX _returnvalue))
    (Rendering|Material|SetScalarParameterValueonMaterials _fx_glow___smoke_sheet "Corona_Opacity" 2e+01)
    (Rendering|Material|SetScalarParameterValueonMaterials _fx_glow___smoke_sheet "Glow_Opacity" 7.0)
    (Rendering|Material|SetVectorParameterValueonMaterials _fx_glow___smoke_sheet "Corona_Colour" "0.450000,0.874998,3.000000")
    (Rendering|Material|SetVectorParameterValueonMaterials _fx_glow___smoke_sheet "Glow_Colour" "0.050000,0.240000,1.000000")
    (Rendering|Components|Light|SetLightColor _treasurelight "(R=0.099899,G=0.270498,B=1.000000,A=1.000000)" true)
    (Rendering|Components|Light|SetIntensity _treasurelight 5.0)
    (Transformation|SetRelativeLocation _treasurelight "0.000000,62.000000,76.000000" false false)
    (Variables|Default|SetP_Loot_Chest_Opened_FX_NS "/Game/Effects/Fort_Effects/World/NewChests/NS_ChestBurst_Rare.NS_ChestBurst_Rare")
    (Variables|Default|SetP_Loot_Chest_Opened_FX 0)
    (CallFunction|FXGlow&SmokeSheetSetup)))

(event Custom|SwapChestAV
  (bind _sound_indicator_component (Variables|Default|GetSound%20Indicator%20Component))
  (Utilities|FlowControl|Delay 0.5)
  ;[1 1 1] Condition
  (if (CallFunction|IsRareChest)
    (Variables|/Script/FortniteGame.FortSoundIndicatorComponent|SetDefaultTintOverride "(R=0.521569,G=0.635294,B=1.000000,A=1.000000)" _sound_indicator_component)
    (Variables|/Script/FortniteGame.FortSoundIndicatorComponent|SetDefaultIconOverride "/Game/UI/Foundation/Textures/Icons/Navigation/AV_PrepperChest_32x32.AV_PrepperChest_32x32" _sound_indicator_component)))

(fn BlueprintCanInteract
  (params (InteractingPawn :type Fort%20Pawn%20Object%20Reference) (InteractionBeingAttempted :type EInteractionBeingAttempted%20Enum :default "FirstInteraction") (InteractionType :type TInteractionType%20Enum :default "IT_NoInteraction"))
  (bind _asfort_player_controller_outpost (Utilities|Casting|CastToFortPlayerControllerOutpost (Pawn|GetController InteractingPawn))
    (:then
      (return (Outpost|IsOutpostOwner _asfort_player_controller_outpost)))
    (:CastFailed
      (return true))))

(fn CleanupWiggleMIDs
  (if (TouchPlatform|IsMobileGame)
    (Components|Timeline|Stop (Variables|Tiered_Chest_Athena|GetMobileOnInteractTL))
    (for _array_element (Variables|Default|GetMIDs)
      (Rendering|Material|SetScalarParameterValue _array_element "WiggleOnInteractAmount" 0.0 :!class /Script/Engine.MaterialInstanceDynamic))))

(fn CreateMobileMIDs
  (if (TouchPlatform|IsMobileGame)
    (for (_ _array_index) (Rendering|Material|GetMaterials (Variables|Debug|GetStaticMeshComponent))
      (bind _returnvalue (CallFunction|GetorCreateMIDOnDefaultMesh _array_index))
      (Utilities|Array|SetArrayElem (Variables|Default|GetMIDs) _array_index _returnvalue true))))

(fn DestroyAdditionalVFXComponents
  ;[1 1 1] Destroy Components in the Additional FX array so that new FX can be added to this list and destroyed later
  (for _array_element (Variables|Default|GetAdditionalVFXToDestroyOnSearch)
    (Utilities|IsValid _array_element
      (:Is%20Valid
        (Components|DestroyComponent _array_element)))))

(fn FXGlow&SmokeSheetInit
  (bind _fx_glow___smoke_sheet (Variables|Default|GetFX%20Glow%20&%20Smoke%20Sheet))
  (Utilities|FlowControl|Sequence
    (:then_0
      (Variables|Effects|SetLastShouldHideSmokeSheetSparkles (Variables|Effects|GetShouldHideSmokeSheetSparkles))
      (Rendering|Material|SetDefaultCustomPrimitiveDataFloat _fx_glow___smoke_sheet 0 (Math|Float|SelectFloat 1.0 0.0 (Variables|Effects|GetShouldHideSmokeSheetSparkles))) ; WPO_HideMesh
      (Rendering|Material|SetDefaultCustomPrimitiveDataFloat _fx_glow___smoke_sheet 2 (Math|Float|SelectFloat 0.0 1.0 (Variables|Effects|GetShouldHideSmokeSheetSparkles))) ; Material Opacity
    )
    (:then_1
      (Variables|Effects|SetLast%20Glow%20Corona%20Size%20Scale (Variables|Effects|GetGlow%20Corona%20Size%20Scale))
      (Rendering|Material|SetDefaultCustomPrimitiveDataFloat (Variables|Default|GetFX%20Glow%20&%20Smoke%20Sheet) 1 (Variables|Effects|GetGlow%20Corona%20Size%20Scale)) ; Scaled_Brightness
    )))

(fn FXGlow&SmokeSheetSetup
  (bind _fx_glow___smoke_sheet (Variables|Default|GetFX%20Glow%20&%20Smoke%20Sheet))
  (Utilities|FlowControl|Sequence
    (:then_0
      (if (!= (Variables|Effects|GetShouldHideSmokeSheetSparkles) (Variables|Effects|GetLastShouldHideSmokeSheetSparkles))
        (Variables|Effects|SetLastShouldHideSmokeSheetSparkles (Variables|Effects|GetShouldHideSmokeSheetSparkles))
        (Rendering|Material|SetCustomPrimitiveDataFloat _fx_glow___smoke_sheet 0 (Math|Float|SelectFloat 1.0 0.0 (Variables|Effects|GetShouldHideSmokeSheetSparkles))) ; WPO_HideMesh
        (Rendering|Material|SetCustomPrimitiveDataFloat _fx_glow___smoke_sheet 2 (Math|Float|SelectFloat 0.0 1.0 (Variables|Effects|GetShouldHideSmokeSheetSparkles))) ; Material Opacity
      ))
    (:then_1
      (if (!= (Variables|Effects|GetGlow%20Corona%20Size%20Scale) (Variables|Effects|GetLast%20Glow%20Corona%20Size%20Scale))
        (Variables|Effects|SetLast%20Glow%20Corona%20Size%20Scale (Variables|Effects|GetGlow%20Corona%20Size%20Scale))
        (Rendering|Material|SetCustomPrimitiveDataFloat (Variables|Default|GetFX%20Glow%20&%20Smoke%20Sheet) 1 (Variables|Effects|GetGlow%20Corona%20Size%20Scale)) ; Scaled_Brightness
      ))))

(fn GetChestType
  (return (select (CallFunction|IsRareChest) (Variables|Default|GetRare%20Container%20Event%20Type) (Variables|Default|GetDefault%20Container%20Event%20Type))))

(fn GetMaxAudibleDistance
  (bind _chest_ambient_sound (Variables|Default|GetChest_Ambient_Sound))
  (bind _sound (Variables|/Script/Engine.AudioComponent|GetSound _chest_ambient_sound))
  (Utilities|IsValid _sound
    (:Is%20Valid
      (return (Variables|/Script/Engine.SoundBase|GetMaxDistance _sound)))
    (:Is%20Not%20Valid
      (return 0.0))))

(fn IsRareChest
  (return (== (Variables|/Script/Engine.StaticMeshComponent|GetStaticMesh (Variables|Debug|GetStaticMeshComponent)) "/Game/Gadgets/Assets/Rare_Chest_Mark_II/Meshes/Rare_Chest_Mark_II.Rare_Chest_Mark_II")))

(fn OnSetCustomDepthStencilValue
  (params (PrimComponents :type Array%20of%20Primitive%20Component%20Object%20References) (bUseCustomDepth :type Boolean) (StencilValue :type Integer))
  (for _array_element PrimComponents
    (Rendering|SetCustomDepthStencilValue _array_element (Variables|GetStencilValue))
    (if (Variables|GetbUseCustomDepth)
      (Rendering|SetRenderCustomDepth _array_element true)
      (Building|BroadcastCustomRenderDepthChange true)))
  (return true))

(fn SetLightVisibility
  (params (Visible :type Boolean))
  (Variables|Default|GetTreasureLight
    (:then
      (Rendering|SetVisibility (Variables|Default|GetTreasureLight) Visible false))
    (:else)))

(fn SetVisibleMobileInteractIcon
  (params (Visible :type Boolean))
  (bind (_returnvalue_roll _returnvalue_pitch _returnvalue_yaw) (Math|Rotator|BreakRotator (Transformation|GetActorRotation)))
  (if Visible
    ;[1 1 1] Create
    (bind _mobileinteracticon (Variables|Default|GetMobileInteractIcon)
      (:then
        (Rendering|SetVisibility _mobileinteracticon true false))
      (:else
        (Variables|Default|GetMobileInteractionMaterial
          (:then
            (bind _returnvalue (AddComponent|NODE_AddStaticMeshComponent-1 :RelativeTransform_Location (Variables|Default|GetMobileInteractIconLocation) :RelativeTransform_Rotation_Yaw (* _returnvalue_yaw -1.0) :RelativeTransform_Scale (Variables|Default|GetMobileInteractIconScale)))
            (bind _output_get (Variables|Default|SetMobileInteractIcon _returnvalue))
            (Rendering|Material|SetMaterial _output_get 0 (Variables|Default|GetMobileInteractionMaterial)))
          (:else))))
    (else
      ;[1 1 1] Destroy
      (bind _mobileinteracticon_2 (Variables|Default|GetMobileInteractIcon)
        (:then
          (Rendering|SetVisibility _mobileinteracticon_2 false false)
          (Components|DestroyComponent _mobileinteracticon_2))
        (:else)))))

(fn ShouldDie
  (params (Damage :type Float%20%28single-precision%29) (EventInstigator :type Controller%20Object%20Reference) (DamageCauser :type Actor%20Object%20Reference) (bIsTest :type Boolean) (EffectContext :type Gameplay%20Effect%20Context%20Handle%20Structure%20%28by%20ref%29))
  (bind _asfort_player_controller_outpost (Utilities|Casting|CastToFortPlayerControllerOutpost EventInstigator)
    (:then
      (return (Outpost|IsOutpostOwner _asfort_player_controller_outpost)))
    (:CastFailed
      (return true))))

(fn ConstructionScript
  (Utilities|FlowControl|Sequence
    (:then_0
      (Networking|CanExecuteCosmeticEvents
        (:True
          (CallFunction|FXGlow&SmokeSheetInit)
          (if (Utilities|String|NotEqualExactly(String) (Game|GetPlatformName) "Windows")
            (Utilities|EffectsQualityCheck 0
              (:Medium
                _)
              (:Low
                _))
            (Components|DestroyComponent (Variables|Default|GetTreasureLight))))))
    (:then_1
      (Utilities|Array|AddUnique (Variables|Actor|GetTags) "Lackluster.Material.TieredChestAthena")
      ;[1 1 1] remove negative scale from all placed chests so vertex shader is correct
      (Transformation|SetActorScale3D (Math|Vector|MakeVector (Math|Float|Absolute(Float) (.x (Transformation|GetActorScale3D))) (Math|Float|Absolute(Float) (.y (Transformation|GetActorScale3D))) (Math|Float|Absolute(Float) (.z (Transformation|GetActorScale3D))))))))
