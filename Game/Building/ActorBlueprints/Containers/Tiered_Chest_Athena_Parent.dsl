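; Parent Blueprint for Athena tiered loot chests, handling ambient glow, chime, opening, reset and Outpost owner-only interaction and damage.
; Looting hides idle cosmetics, spawns rarity-tinted VFX and drives a vertex-shader lid animation through CODE_Timestamp; loot reset reverses it.
; The non-cosmetic path allows building over opened chests and sends minigame container-open events.
; Mobile adds MID-driven wiggle and outline plus a scaling interact icon.
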(event EventBeginPlay
  (Networking|CanExecuteCosmeticEvents
    (:True
      ;[1 1 1] Begin Play
      (Default|FXGlow&SmokeSheetSetup (Variables|FX|GetShould%20Hide%20Smoke%20Sheet%20&%20Sparkles) (Variables|FX|GetGlow%20Corona%20Brightness%20Scale) (Variables|FX|GetGlow%20Corona%20Size%20Scale))
      (CallFunction|CreateMobileMIDs)
      (if (Variables|Container|GetbAlreadySearched)
        (CallFunction|SetLightVisibility false)
        (Audio|Components|Audio|Stop (Variables|Default|GetChest_Ambient_Sound))
        (Variables|Default|SetbOpened true)
        (else
          (Rendering|SetVisibility (Variables|Default|GetCustom%20Smoke%20Sheet) true false)
          (CallFunction|SetLightVisibility true)
          (if (Variables|Default|GetEnableAudioIndicator)
            (CallFunction|StartAudiovisualIndicator (Variables|Default|GetChest_Ambient_Sound) "Chest" (Variables|Default|GetChestChimeVisualUpdate))
            (CallFunction|ResetChestWPOForPIE)))))))

(event Custom|Mobile_OnInteract
  ;[1 1 1] Mobile Wiggle on Interact
  (Networking|CanExecuteCosmeticEvents
    (:True
      (bind (_direction _lerp) (Timeline|MobileOnInteractTL Play)
        (:Update
          (for _array_element (Variables|Default|GetMIDs)
            (Rendering|Material|SetScalarParameterValue _array_element "WiggleOnInteractAmount" (Math|Float|Lerp 0.0 (Variables|Default|GetMobileWiggleAmount) _lerp) :!class /Script/Engine.MaterialInstanceDynamic)))))))

(event Custom|Mobile_Outline_On
  ;[1 1 1] Mobile Outline
  (for _array_element (Variables|Default|GetMIDs)
    (Rendering|Material|SetScalarParameterValue _array_element "Outline_Control" 1.0 :!class /Script/Engine.MaterialInstanceDynamic)
    (Rendering|Material|SetScalarParameterValue _array_element "Pulse_Control" 1.0 :!class /Script/Engine.MaterialInstanceDynamic)
    (Rendering|Material|SetVectorParameterValue _array_element "Color" "(R=1.000000,G=1.000000,B=1.000000,A=1.000000)" :!class /Script/Engine.MaterialInstanceDynamic)))

(event Container|EventEnableBacchusHighlight
  ;[1 1 1] Mobile highlighting
  (if (not (Variables|Container|GetbAlreadySearched))
    (Networking|CanExecuteCosmeticEvents
      (:True
        (CallFunction|SetVisibleMobileInteractIcon true)
        (bind (_direction _lerpobject _lerpinteactoicon) (Timeline|MobileSelectedTL Play)
          (:Update
            (bind _mobileinteracticon (Variables|Default|GetMobileInteractIcon)
              (:then
                (Transformation|SetRelativeScale3D _mobileinteracticon (Math|Vector|Lerp(Vector) "0, 0, 0" (Variables|Default|GetMobileInteractIconScale) _lerpinteactoicon)))
              (:else)))
          (:Finished
            (switch Utilities|FlowControl|Switch|SwitchonETimelineDirection _direction
              (:Backward
                (CallFunction|SetVisibleMobileInteractIcon false)))))))))

(event Custom|Mobile_OnDisconnect
  ;[1 1 1] Mobile Wiggle on Interact
  (Networking|CanExecuteCosmeticEvents
    (:True
      (Timeline|MobileOnInteractTL Reverse))))

(event Custom|Mobile_Outline_Off
  ;[1 1 1] Mobile Outline
  (for _array_element (Variables|Default|GetMIDs)
    (Rendering|Material|SetScalarParameterValue _array_element "Outline_Control" 0.0 :!class /Script/Engine.MaterialInstanceDynamic)
    (Rendering|Material|SetScalarParameterValue _array_element "Pulse_Control" 0.0 :!class /Script/Engine.MaterialInstanceDynamic)))

(event Container|EventDisableBacchusHighlight
  ;[1 1 1] Mobile highlighting
  (if (TouchPlatform|IsMobileGame)
    (Networking|CanExecuteCosmeticEvents
      (:True
        (Timeline|MobileSelectedTL Reverse)))))

(event EventOnBeginSearch
  ;[1 1 1] Mobile highlighting
  (if (Variables|Container|GetbAlreadySearched)
    _
    (elif (TouchPlatform|IsMobileGame)
      (Networking|CanExecuteCosmeticEvents
        (:True
          (Timeline|MobileSelectedTL Reverse))))))

(event EventOnSetSearched
  (Networking|CanExecuteCosmeticEvents
    (:True
      (CallFunction|HideChestOpenableCosmetics))
    (:False
      (Variables|Building|SetbDestroyOnPlayerBuildingPlacement true) ; Allows players to build over the opened chest mesh (but not when there is loot in there)
    )))

(event EventDestroyed
  (Networking|CanExecuteCosmeticEvents
    (:True
      (CallFunction|HideChestOpenableCosmetics))
    (:False
      (Variables|Building|SetbDestroyOnPlayerBuildingPlacement true) ; Allows players to build over the opened chest mesh (but not when there is loot in there)
    )))

(event EventOnLoot
  (Networking|CanExecuteCosmeticEvents
    (:True
      (Utilities|FlowControl|Sequence
        (:then_0
          (CallFunction|HideChestOpenableCosmetics))
        (:then_1
          (CallFunction|SpawnChestOpenVFX))
        (:then_2
          (CallFunction|PlayChestOpeningTimeline))))
    (:False
      (Variables|Building|SetbDestroyOnPlayerBuildingPlacement true) ; Allows players to build over the opened chest mesh (but not when there is loot in there)
      (Default|CallContainerOpened)
      (Utilities|IsValid (Minigame|GetMinigame self :!class /Script/FortniteGame.MinigameBlueprintLibrary)
        (:Is%20Valid
          (Networking|SwitchHasAuthority
            (:Authority
              (Class|LootControllerFunctionLibrary|MakeandSendOpenContainerEvent self (Variables|Default|GetContainer%20Type)))))))))

(event Custom|HideChestOpenableCosmetics
  (Utilities|FlowControl|DoOnce false
    (:Completed
      (Rendering|SetVisibility (Variables|Default|GetCustom%20Smoke%20Sheet) false false)
      (Components|DestroyComponent (Variables|Default|GetItemMesh))
      (Components|DestroyComponent (Variables|Default|GetCustom%20Niagara%20Effects))
      (CallFunction|SetLightVisibility false)
      ;[1 1 1] Prevents double triggering of opening sound for the local player
      (bind _returnvalue (Container|GetSearchingPawn))
      (Utilities|IsValid _returnvalue
        (:Is%20Valid
          (if (Pawn|IsLocallyControlled _returnvalue)
            (bind _chest_ambient_sound (Variables|Default|GetChest_Ambient_Sound :!label _label)
              (:then
                (Audio|Components|Audio|FadeOut _chest_ambient_sound 0.5 0.0 "Linear" :!class /Script/Engine.AudioComponent)
                (Variables|Default|SetbOpened true))
              (:else))
            (else
              (CallFunction|AudioonOpened)))
          (CallFunction|StopAudiovisualIndicator)
          (CallFunction|CleanupWiggleMIDs)
          (Components|Timeline|Stop (Variables|Tiered_Chest_Athena_Parent|GetMobileSelectedTL))
          (CallFunction|SetVisibleMobileInteractIcon false)
          (Rendering|SetRenderCustomDepth (Variables|Debug|GetStaticMeshComponent) false)
          (Building|BroadcastCustomRenderDepthChange false)
          (Default|DestroyAdditionalVFXComponents)
          ;[1 1 1] Prevents sound library tags to get applied after the chest is opened
          (if (Variables|Default|GetbOpened)
            (SoundLibrary|DisableEventsforActor (WorldSubsystems|GetSoundLibrarySubsystem) self)))
        (:Is%20Not%20Valid
          _))
      (goto _label))))

(event Custom|ResetOpenableCosmetics
  (Utilities|FlowControl|DoOnce false
    (:Completed
      (Rendering|SetVisibility (Variables|Default|GetCustom%20Smoke%20Sheet) false false)
      (Components|DestroyComponent (Variables|Default|GetItemMesh))
      (Components|DestroyComponent (Variables|Default|GetCustom%20Niagara%20Effects))
      (CallFunction|SetLightVisibility false)
      ;[1 1 1] Prevents double triggering of opening sound for the local player
      (bind _returnvalue (Container|GetSearchingPawn))
      (Utilities|IsValid _returnvalue
        (:Is%20Valid
          (if (Pawn|IsLocallyControlled _returnvalue)
            (bind _chest_ambient_sound (Variables|Default|GetChest_Ambient_Sound :!label _label)
              (:then
                (Audio|Components|Audio|FadeOut _chest_ambient_sound 0.5 0.0 "Linear" :!class /Script/Engine.AudioComponent)
                (Variables|Default|SetbOpened true))
              (:else))
            (else
              (CallFunction|AudioonOpened)))
          (CallFunction|StopAudiovisualIndicator)
          (CallFunction|CleanupWiggleMIDs)
          (Components|Timeline|Stop (Variables|Tiered_Chest_Athena_Parent|GetMobileSelectedTL))
          (CallFunction|SetVisibleMobileInteractIcon false)
          (Rendering|SetRenderCustomDepth (Variables|Debug|GetStaticMeshComponent) false)
          (Building|BroadcastCustomRenderDepthChange false)
          (Default|DestroyAdditionalVFXComponents)
          ;[1 1 1] Prevents sound library tags to get applied after the chest is opened
          (if (Variables|Default|GetbOpened)
            (SoundLibrary|DisableEventsforActor (WorldSubsystems|GetSoundLibrarySubsystem) self)))
        (:Is%20Not%20Valid
          _))
      (goto _label))))

(event Custom|PlayChestOpeningTimeline
  (Utilities|FlowControl|DoOnce false
    (:Completed
      (bind (_direction _time) (Timeline|Timeline_0 Play)
        (:Update
          (Rendering|Material|SetScalarParameterValueonMaterials (Variables|Debug|GetStaticMeshComponent) "CODE_Timestamp" _time))
        (:Finished
          (if (Variables|Default|GetbClosing)
            (Variables|Default|SetbClosing false)
            (CallFunction|OnClosingComplete)))))))

(event Custom|SpawnChestOpenVFX
  (bind _returnvalue (Effects|SpawnEmitteratLocation (Variables|Default|GetP_Loot_Chest_Opened_FX) (Transformation|GetActorLocation) (Math|Rotator|MakeRotator (.roll (Transformation|GetActorRotation)) (+ (.pitch (Transformation|GetActorRotation)) (Variables|Default|GetOpenFX_Rotation)) (.yaw (Transformation|GetActorRotation)))))
  (Variables|Default|SetLoot_Effect _returnvalue)
  (bind _returnvalue_2 (Fortnite|GetRarityData (Variables|Container|GetHighestRarity)))
  (Effects|Components|ParticleSystem|SetVectorParameter (Variables|Default|GetLoot_Effect) "color" (Math|Conversions|ToVector(LinearColor) (.Color2 (Utilities|Struct|BreakFortRarityItemData _returnvalue_2)))))

(event Custom|ResetOpeningTimeline
  (Utilities|FlowControl|Sequence
    (:then_0
      (Utilities|FlowControl|DoOnce false
        (:Completed
          (Timeline|Timeline_0 Play))))
    (:then_1
      (Timeline|Timeline_0 ReverseFromEnd))
    (:then_2
      (Variables|Default|SetbClosing true))))

(event EventOnLootReset
  (Networking|CanExecuteCosmeticEvents
    (:True
      (CallFunction|ResetOpeningTimeline))))

(event EventOnDeathPlayEffects
  (params (Damage :type Float%20%28single-precision%29) (DamageTags :type Gameplay%20Tag%20Container%20Structure%20%28by%20ref%29) (Momentum :type Vector :default "0, 0, 0") (HitInfo :type Hit%20Result%20Structure%20%28by%20ref%29) (InstigatedBy :type Fort%20Pawn%20Object%20Reference) (DamageCauser :type Actor%20Object%20Reference) (EffectContext :type Gameplay%20Effect%20Context%20Handle%20Structure))
  (CallFunction|AudioonOpened))

(event Custom|OnClosingComplete
  (Utilities|FlowControl|Sequence
    (:then_0
      (CallFunction|ResetOpenableCosmetics))
    (:then_1
      (CallFunction|SetLightVisibility true)
      (bind _chest_ambient_sound (Variables|Default|GetChest_Ambient_Sound)
        (:then
          (Audio|Components|Audio|FadeIn _chest_ambient_sound 0.0 1.0 0.0 "Linear" :!class /Script/Engine.AudioComponent)
          (CallFunction|StartAudiovisualIndicator (Variables|Default|GetChest_Ambient_Sound) "Chest" (Variables|Default|GetChestChimeVisualUpdate)))
        (:else)))
    (:then_2
      (Default|FXGlow&SmokeSheetSetup false))
    (:then_3
      (Rendering|SetVisibility (Variables|Default|GetFX%20Glow%20&%20Smoke%20Sheet) true false))))

(event Custom|AudioOnOpened
  (if (not (Variables|Default|GetbOpened))
    (Audio|Components|Audio|FadeOut (Variables|Default|GetChest_Ambient_Sound) 0.5 0.0 "Linear" :!class /Script/Engine.AudioComponent)
    (SoundLibrary|PlaySoundSimple (WorldSubsystems|GetSoundLibrarySubsystem) self (Utilities|Struct|MakeSimpleSoundSettings "(TagName=\"SoundLibrary.Item.Interact.Complete\")") :!class /Script/SoundLibrary.SoundLibrarySubsystem)
    (bind _open_chest_sound (Variables|Sound|GetOpen%20Chest%20Sound)
      (:then
        (bind _chestopensourceeffectoverride (Variables|Sound|GetChestOpenSourceEffectOverride)
          (:then
            (Variables|/Script/Engine.SoundBase|SetSourceEffectChain _chestopensourceeffectoverride _open_chest_sound))
          (:else
            (Variables|/Script/Engine.SoundBase|SetSourceEffectChain 0 _open_chest_sound)))
        (Audio|PlaySoundatLocation (Variables|Sound|GetOpen%20Chest%20Sound) (Transformation|GetActorLocation)))
      (:else
        _))
    (Variables|Default|SetbOpened true)))

(fn BlueprintCanInteract
  (params (InteractingPawn :type Fort%20Pawn%20Object%20Reference) (InteractionBeingAttempted :type EInteractionBeingAttempted%20Enum :default "FirstInteraction") (InteractionType :type TInteractionType%20Enum :default "IT_NoInteraction"))
  (bind _asfort_player_controller_outpost (Utilities|Casting|CastToFortPlayerControllerOutpost (Pawn|GetController InteractingPawn))
    (:then
      (return (Outpost|IsOutpostOwner _asfort_player_controller_outpost)))
    (:CastFailed
      (return true))))

(fn BuildExposedAssetProviderVersion(Implementation)
  (params (Version :type Fort%20Exposed%20Asset%20Provider%20Version%20Structure%20%28by%20ref%29))
  (|CallParentFunction Version)
  (ExposedAssetProvider|AddVersion Version 1)
  (return))

(fn CleanupWiggleMIDs
  (if (TouchPlatform|IsMobileGame)
    (Components|Timeline|Stop (Variables|Tiered_Chest_Athena_Parent|GetMobileOnInteractTL))
    (for _array_element (Variables|Default|GetMIDs)
      (Rendering|Material|SetScalarParameterValue _array_element "WiggleOnInteractAmount" 0.0 :!class /Script/Engine.MaterialInstanceDynamic))))

(fn CollectAssetReferencestoAllow(Implementation)
  (params (Collector :type Fort%20Exposed%20Asset%20Reference%20Collector%20Structure%20%28by%20ref%29))
  (|CallParentFunction Collector)
  (bind _returnvalue (Fortnite|Utilities|ResolveComponentFromActor 0 (Variables|Default|GetFX%20Glow%20&%20Smoke%20Sheet)))
  (for _array_element (Rendering|Material|GetMaterials _returnvalue)
    (ExposedAssetProvider|CollectIndirectAssetReference Collector (Utilities|ResolveSoftReference _array_element)))
  (return))

(fn CreateMobileMIDs
  (if (TouchPlatform|IsMobileGame)
    (for (_ _array_index) (Rendering|Material|GetMaterials (Variables|Debug|GetStaticMeshComponent))
      (bind _returnvalue (CallFunction|GetorCreateMIDOnDefaultMesh _array_index))
      (Utilities|Array|SetArrayElem (Variables|Default|GetMIDs) _array_index _returnvalue true))))

(fn DestroyAdditionalVFXComponents
  ;[1 1 1] Destroy Components in the Additional FX array so that new FX can be added to this list and destroyed later
  (for _array_element (Variables|Default|GetAdditionalVFXToDestroyOnSearch)
    (Utilities|IsValid _array_element
      (:Is%20Valid
        (Components|DestroyComponent _array_element)))))

(fn FXGlow&SmokeSheetSetup
  (params (Should%20Hide%20Smoke%20Sheet%20&%20Sparkles :type Boolean) (Glow%20Corona%20Brightness%20Scale :type Float%20%28double-precision%29) (Glow%20Corona%20Size%20Scale :type Float%20%28double-precision%29))
  (bind _fx_glow___smoke_sheet (Variables|Default|GetFX%20Glow%20&%20Smoke%20Sheet))
  (Rendering|Material|SetCustomPrimitiveDataFloat _fx_glow___smoke_sheet 0 (Math|Float|SelectFloat 1.0 0.0 Should%20Hide%20Smoke%20Sheet%20&%20Sparkles)) ; WPO_HideMesh
  (Rendering|Material|SetCustomPrimitiveDataFloat _fx_glow___smoke_sheet 1 Glow%20Corona%20Size%20Scale) ; Scaled_Brightness
  (Rendering|Material|SetDefaultCustomPrimitiveDataFloat _fx_glow___smoke_sheet 2 (Math|Float|SelectFloat 0.0 1.0 (Variables|FX|GetShould%20Hide%20Smoke%20Sheet%20&%20Sparkles))) ; Material Opacity
)

(fn GetMaxAudibleDistance
  (bind _chest_ambient_sound (Variables|Default|GetChest_Ambient_Sound))
  (bind _sound (Variables|/Script/Engine.AudioComponent|GetSound _chest_ambient_sound))
  (Utilities|IsValid _sound
    (:Is%20Valid
      (return (Variables|/Script/Engine.SoundBase|GetMaxDistance _sound)))
    (:Is%20Not%20Valid
      (return 0.0))))

(fn OnSetCustomDepthStencilValue
  (params (PrimComponents :type Array%20of%20Primitive%20Component%20Object%20References) (bUseCustomDepth :type Boolean) (StencilValue :type Integer))
  (for _array_element PrimComponents
    (Rendering|SetCustomDepthStencilValue _array_element (Variables|GetStencilValue))
    (Rendering|SetRenderCustomDepth _array_element (Variables|GetbUseCustomDepth))
    (Building|BroadcastCustomRenderDepthChange true))
  (return true))

(fn ResetChestWPOForPIE
  (if (Editor|InEditororPIE)
    (if (not (Variables|Container|GetbAlreadySearched))
      (Rendering|Material|SetScalarParameterValueonMaterials (Variables|Debug|GetStaticMeshComponent) "CODE_Timestamp" 0.0))))

(fn SetLightVisibility
  (params (Visible :type Boolean))
  (Variables|Default|GetTreasureLight
    (:then
      (Rendering|SetVisibility (Variables|Default|GetTreasureLight) Visible false))
    (:else)))

(fn SetVisibleMobileInteractIcon
  (params (Visible :type Boolean))
  (bind (_returnvalue_roll _returnvalue_pitch _returnvalue_yaw) (Math|Rotator|BreakRotator (Transformation|GetActorRotation)))
  (if Visible
    ;[1 1 1] Create
    (bind _mobileinteracticon (Variables|Default|GetMobileInteractIcon)
      (:then
        (Rendering|SetVisibility _mobileinteracticon true false))
      (:else
        (Variables|Default|GetMobileInteractionMaterial
          (:then
            (bind _returnvalue (AddComponent|NODE_AddStaticMeshComponent-1 :RelativeTransform_Location (Variables|Default|GetMobileInteractIconLocation) :RelativeTransform_Rotation_Yaw (* _returnvalue_yaw -1.0) :RelativeTransform_Scale (Variables|Default|GetMobileInteractIconScale)))
            (bind _output_get (Variables|Default|SetMobileInteractIcon _returnvalue))
            (Rendering|Material|SetMaterial _output_get 0 (Variables|Default|GetMobileInteractionMaterial)))
          (:else))))
    (else
      ;[1 1 1] Destroy
      (bind _mobileinteracticon_2 (Variables|Default|GetMobileInteractIcon)
        (:then
          (Rendering|SetVisibility _mobileinteracticon_2 false false)
          (Components|DestroyComponent _mobileinteracticon_2))
        (:else)))))

(fn ShouldDie
  (params (Damage :type Float%20%28single-precision%29) (EventInstigator :type Controller%20Object%20Reference) (DamageCauser :type Actor%20Object%20Reference) (bIsTest :type Boolean) (EffectContext :type Gameplay%20Effect%20Context%20Handle%20Structure%20%28by%20ref%29))
  (bind _asfort_player_controller_outpost (Utilities|Casting|CastToFortPlayerControllerOutpost EventInstigator)
    (:then
      (return (Outpost|IsOutpostOwner _asfort_player_controller_outpost)))
    (:CastFailed
      (return true))))

(fn ConstructionScript
  (Default|FXGlow&SmokeSheetSetup (Variables|FX|GetShould%20Hide%20Smoke%20Sheet%20&%20Sparkles) (Variables|FX|GetGlow%20Corona%20Brightness%20Scale) (Variables|FX|GetGlow%20Corona%20Size%20Scale))
  (Utilities|FlowControl|Sequence
    (:then_0
      (Networking|CanExecuteCosmeticEvents
        (:True
          (if (Utilities|String|NotEqualExactly(String) (Game|GetPlatformName) "Windows")
            (Utilities|EffectsQualityCheck 0
              (:Medium
                _)
              (:Low
                _))
            (Components|DestroyComponent (Variables|Default|GetTreasureLight))))))
    (:then_1
      ;[1 1 1] remove negative scale from all placed chests so vertex shader is correct
      (Transformation|SetActorScale3D (Math|Vector|MakeVector (Math|Float|Absolute(Float) (.x (Transformation|GetActorScale3D))) (Math|Float|Absolute(Float) (.y (Transformation|GetActorScale3D))) (Math|Float|Absolute(Float) (.z (Transformation|GetActorScale3D))))))
    (:then_2
      (CallFunction|ResetChestWPOForPIE))))
