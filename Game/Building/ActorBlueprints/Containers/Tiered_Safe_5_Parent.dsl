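; Parent Blueprint for a tiered safe loot container that handles its closed-state effect, treasure light and opening effects.
; BeginPlay turns the effect and light on unless the safe is already searched; searching, looting or destruction turns them off.
; Looting also spawns a chest particle tinted by the highest rarity's color and plays an open sound once.
; The construction script runs cosmetic and effects-quality checks around destroying the light component, so visibility changes check its validity.
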(event EventBeginPlay
  (if (Variables|Container|GetbAlreadySearched)
    (CallFunction|SetLightVisibility false)
    (else
      (Components|Activation|Activate (Variables|Default|GetClosed_Chest) false)
      (CallFunction|SetLightVisibility true))))

(event EventOnSetSearched
  (Components|Activation|Deactivate (Variables|Default|GetClosed_Chest))
  (CallFunction|SetLightVisibility false))

(event EventDestroyed
  (Components|Activation|Deactivate (Variables|Default|GetClosed_Chest))
  (CallFunction|SetLightVisibility false))

(event EventOnLoot
  (bind _returnvalue (Transformation|GetActorLocation))
  (Utilities|FlowControl|Sequence
    (:then_0
      (Components|Activation|Deactivate (Variables|Default|GetClosed_Chest))
      (CallFunction|SetLightVisibility false))
    (:then_1
      (bind _returnvalue_2 (Effects|SpawnEmitteratLocation "/Game/Effects/Fort_Effects/World/Chests/P_Loot_Chest_T1.P_Loot_Chest_T1" _returnvalue (Math|Rotator|MakeRotator (.roll (Transformation|GetActorRotation)) (+ (.pitch (Transformation|GetActorRotation)) 90) (.yaw (Transformation|GetActorRotation)))))
      (Variables|Default|SetLoot_Effect _returnvalue_2)
      (bind _returnvalue_3 (Fortnite|GetRarityData (Variables|Container|GetHighestRarity)))
      (Effects|Components|ParticleSystem|SetVectorParameter (Variables|Default|GetLoot_Effect) "color" (Math|Conversions|ToVector(LinearColor) (.Color2 (Utilities|Struct|BreakFortRarityItemData _returnvalue_3))))
      (Utilities|FlowControl|DoOnce false
        (:Completed
          (Audio|PlaySoundatLocation "/Game/Sounds/Foley_Loot/Containers/Treasure_Chest/Tiered_Chest_Open_T01_Cue.Tiered_Chest_Open_T01_Cue" _returnvalue))))))

(fn SetLightVisibility
  (params (Active :type Boolean))
  (Utilities|IsValid (Variables|Default|GetTrasureLight)
    (:Is%20Valid
      (Rendering|SetVisibility (Variables|Default|GetTrasureLight) Active false))))

(fn ConstructionScript
  (Networking|CanExecuteCosmeticEvents
    (:True
      (Utilities|EffectsQualityCheck 0
        (:Medium
          _)
        (:Low
          _))
      (Components|DestroyComponent (Variables|Default|GetTrasureLight)))))
