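; Blueprint for a tiered safe container in the Athena (Battle Royale) mode, a non-physics variant placed in levels.
; Its construction script runs the parent's, then adds a tiered-chest material tag to the actor's tags without duplicates.
; Running this at construction ensures already-placed instances receive the tag, so they get the correct chrome material.
; The event graph contains only empty or disabled stub events, leaving all other behavior to the parent class.
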(event EventBeginPlay
  (|CallParentFunction) ; This node is disabled and will not be called.\nDrag off pins to build functionality.
)

(event Collision|EventActorBeginOverlap
  (params (OtherActor :type Actor%20Object%20Reference)))

(event EventTick
  (params (DeltaSeconds :type Float%20%28single-precision%29)))

(event EventDestroyed)

(fn ConstructionScript
  ;[0.32 0.9402 1] Make sure actors that are already placed in levels get this tag so that they get the right chrome material added to them.
  (|CallParentFunction)
  (Utilities|Array|AddUnique (Variables|Actor|GetTags) "Lackluster.Material.TieredChestAthena"))
