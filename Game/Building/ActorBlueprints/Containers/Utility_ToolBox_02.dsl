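; Utility_ToolBox_02 is a toolbox container building actor Blueprint that reacts to being looted.
; On the loot event, it hides the particle system components attached to its static mesh component.
; The construction script is empty, and no other logic is defined in this Blueprint.
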(event EventOnLoot
  (Utilities|HideParticleSystemComponentsAttachedtoMesh (Variables|Debug|GetStaticMeshComponent)))

(fn ConstructionScript)
